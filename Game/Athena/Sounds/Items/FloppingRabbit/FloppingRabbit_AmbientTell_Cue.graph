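# Sound cue for the FloppingRabbit item's ambient tell in Athena, playing flopping sounds.
# A Random node picks one of four equally weighted FlopperSpawn ambient flop waves, and a Modulator varies its pitch and volume.
# It plays quietly in the outdoor ambient sound class, with overridden attenuation using logarithmic falloff in place of the referenced FloppingRabbit hit attenuation asset.
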SoundCue /Game/Athena/Sounds/Items/FloppingRabbit/FloppingRabbit_AmbientTell_Cue
root=Modulator_0
nodes [
 {id=Modulator_0, children [Random_0], properties {pitchMax=1.1, pitchMin=1, volumeMin=0.7}}
 {id=Random_0, children [WavePlayer_1,WavePlayer_2,WavePlayer_3,WavePlayer_6], properties {weights [1,1,1,1]}}
 {id=WavePlayer_1, properties {soundWaveAssetPtr=&/FlopperSpawn_AmbientFlop_04}}
 {id=WavePlayer_2, properties {soundWaveAssetPtr=&/FlopperSpawn_AmbientFlop_01}}
 {id=WavePlayer_3, properties {soundWaveAssetPtr=&/FlopperSpawn_AmbientFlop_02}}
 {id=WavePlayer_6, properties {soundWaveAssetPtr=&/FlopperSpawn_AmbientFlop_03}}
]
properties {
 attenuationSettings=/Game/Sounds/Attenuation/Fort_FloppingRabbitHit_Att
 bOverrideAttenuation=true
 maxDistance=7701.1201
 soundClassObject=&Fort_Ambient_Outdoors_Athena
 volumeMultiplier=0.251
 attenuationOverrides {distanceAlgorithm=Logarithmic, stereoSpread=1297.9291, attenuationShapeExtents {x=4101.1201}}
}
