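# Sound cue for the drop sound of the Flopping Rabbit item in Athena.
# A single wave player feeds a modulator that randomizes pitch slightly on each play while holding volume constant.
# It plays in the loot collect sound class with weapon-drop attenuation and a slightly reduced cue volume.
# It references the shared weapon-drop concurrency set, named for stopping the oldest sound.
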SoundCue /Game/Athena/Sounds/Items/FloppingRabbit/FloppingRabbit_Drop_Cue
root=Modulator_1
nodes [
 {id=Modulator_1, children [WavePlayer_5], properties {pitchMax=1.2, pitchMin=0.9, volumeMax=1, volumeMin=1}}
 {id=WavePlayer_5, properties {soundWaveAssetPtr=&/FloppingRabbit_Drop_01}}
]
properties {
 attenuationSettings=&Fort_Weapon_Drop_Att
 maxDistance=3000
 soundClassObject=&Fort_Loot_Collect
 volumeMultiplier=0.85
 concurrencyOverrides {resolutionRule=PreventNew}
 concurrencySet [&Weapon_Drop_StopOldest]
}
