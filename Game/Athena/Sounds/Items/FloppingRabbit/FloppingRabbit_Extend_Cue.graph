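# Sound cue for the FloppingRabbit item's extend sound, categorised under the melee weapon sound class.
# A single wave player feeds the extend sample into a modulator at the root, with the overall cue volume reduced.
# Spatial falloff comes from the close-range pickaxe attenuation settings.
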SoundCue /Game/Athena/Sounds/Items/FloppingRabbit/FloppingRabbit_Extend_Cue
root=Modulator_0
nodes [
 {id=Modulator_0, children [WavePlayer_0]}
 {id=WavePlayer_0, properties {soundWaveAssetPtr=&/FloppingRabbit_Extend_01}}
]
properties {
 attenuationSettings=/Game/Sounds/Attenuation/Fort_Weapon_Pickaxe_Close_Att
 maxDistance=1536
 soundClassObject=&Fort_WeaponMelee
 volumeMultiplier=0.5
}
