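# Sound cue for the Flopping Rabbit item's impact sound.
# It randomly picks one of three equally weighted impact waves, then passes it through a modulator that slightly raises pitch at random for variation.
# It plays in the gadgets and items sound class with a dedicated Flopping Rabbit hit attenuation and a slightly boosted volume.
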SoundCue /Game/Athena/Sounds/Items/FloppingRabbit/FloppingRabbit_Impact_Cue
root=Modulator_0
nodes [
 {id=Modulator_0, children [Random_0], properties {pitchMax=1.1, pitchMin=1}}
 {id=Random_0, children [WavePlayer_1,WavePlayer_2,WavePlayer_3], properties {weights [1,1,1]}}
 {id=WavePlayer_1, properties {soundWaveAssetPtr=&/FloppingRabbit_Impact_01}}
 {id=WavePlayer_2, properties {soundWaveAssetPtr=&/FloppingRabbit_Impact_02}}
 {id=WavePlayer_3, properties {soundWaveAssetPtr=&/FloppingRabbit_Impact_03}}
]
properties {
 attenuationSettings=/Game/Sounds/Attenuation/Fort_FloppingRabbitHit_Att
 maxDistance=5300
 soundClassObject=&Fort_Gadgets_Items
 volumeMultiplier=1.2
}
