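# FloppingRabbit_ItemSpawn_Cue is a sound cue for the spawn of the FloppingRabbit item, classed under gadget item sounds.
# A Random node picks with equal weight between two wave players, both reusing third-person player water swim-sprint recordings.
# The cue plays them pitched up and slightly quieter, with spatial falloff from the FloppingRabbit hit attenuation asset.
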SoundCue /Game/Athena/Sounds/Items/FloppingRabbit/FloppingRabbit_ItemSpawn_Cue
root=Random_1
nodes [
 {id=Random_1, children [WavePlayer_0,WavePlayer_4], properties {weights [1,1]}}
 {id=WavePlayer_0, properties {soundWaveAssetPtr=/Game/Athena/Sounds/Water/Interaction/Water_Player_Swim_Sprint_3P_10}}
 {id=WavePlayer_4, properties {soundWaveAssetPtr=/Game/Athena/Sounds/Water/Interaction/Water_Player_Swim_Sprint_3P_09}}
]
properties {
 attenuationSettings=/Game/Sounds/Attenuation/Fort_FloppingRabbitHit_Att
 maxDistance=5300
 pitchMultiplier=2
 soundClassObject=&Fort_Gadgets_Items
 volumeMultiplier=0.8
}
