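# Sound cue for the landing sound of the FloppingRabbit item, classed under loot collection sounds.
# A single landing wave plays through a modulator that randomizes pitch slightly while holding volume constant.
# Overridden attenuation gives it a short falloff range, and a concurrency override caps simultaneous copies, blocking new plays beyond the limit.
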SoundCue /Game/Athena/Sounds/Items/FloppingRabbit/FloppingRabbit_Landed_Cue
root=Modulator_0
nodes [
 {id=Modulator_0, children [WavePlayer_5], properties {pitchMax=1.2, pitchMin=0.9, volumeMax=1, volumeMin=1}}
 {id=WavePlayer_5, properties {soundWaveAssetPtr=&/FloppingRabbit_Land_01}}
]
properties {
 attenuationSettings=&Foley_Attenuation_Basic_Att
 bOverrideAttenuation=true
 bOverrideConcurrency=true
 maxDistance=1000
 soundClassObject=&Fort_Loot_Collect
 volumeMultiplier=0.9
 attenuationOverrides {falloffDistance=1000, attenuationShapeExtents {x=0}}
 concurrencyOverrides {maxCount=3, resolutionRule=PreventNew}
}
