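# Sound cue that plays when a Flopping Rabbit item is picked up.
# A single pickup wave runs through a modulator that randomly raises pitch slightly for variation, with volume held fixed.
# It uses the weapon pickup attenuation and loot-collect sound class, and references a shared pickup concurrency set that, by its name, drops the oldest sound.
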SoundCue /Game/Athena/Sounds/Items/FloppingRabbit/FloppingRabbit_Pickup_Cue
root=Modulator_0
nodes [
 {id=Modulator_0, children [WavePlayer_5], properties {pitchMax=1.1, pitchMin=1, volumeMax=1, volumeMin=1}}
 {id=WavePlayer_5, properties {soundWaveAssetPtr=&/FloppingRabbit_Pickup_01}}
]
properties {
 attenuationSettings=&Fort_Weapon_Pickup_Att
 maxDistance=3500
 soundClassObject=&Fort_Loot_Collect
 volumeMultiplier=1
 concurrencyOverrides {maxCount=1, resolutionRule=PreventNew}
 concurrencySet [&PickupMaximum_Oldest]
}
