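# Sound cue for the FloppingRabbit item's Ready sound, routed through the melee weapon sound class.
# A single wave player feeds a Modulator that randomizes pitch on each play while holding volume steady.
# The cue lowers overall pitch and volume and uses the close-range pickaxe attenuation asset for spatial falloff.
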SoundCue /Game/Athena/Sounds/Items/FloppingRabbit/FloppingRabbit_Ready_Cue
root=Modulator_0
nodes [
 {id=Modulator_0, children [WavePlayer_0], properties {pitchMax=1.1, volumeMax=1, volumeMin=1}}
 {id=WavePlayer_0, properties {soundWaveAssetPtr=&/FloppingRabbit_Ready}}
]
properties {
 attenuationSettings=/Game/Sounds/Attenuation/Fort_Weapon_Pickaxe_Close_Att
 maxDistance=1536
 pitchMultiplier=0.9
 soundClassObject=&Fort_WeaponMelee
 volumeMultiplier=0.365
}
