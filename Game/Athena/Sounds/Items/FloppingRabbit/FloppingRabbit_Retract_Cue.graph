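# Sound cue for the retract sound of the FloppingRabbit item.
# A single wave player feeds FloppingRabbit_Retract_01 through a Modulator left at default settings for slight variation.
# The cue plays at reduced volume in the melee weapon sound class, using a non-binaural gadget-activate attenuation asset.
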SoundCue /Game/Athena/Sounds/Items/FloppingRabbit/FloppingRabbit_Retract_Cue
root=Modulator_0
nodes [
 {id=Modulator_0, children [WavePlayer_0]}
 {id=WavePlayer_0, properties {soundWaveAssetPtr=&/FloppingRabbit_Retract_01}}
]
properties {
 attenuationSettings=/Game/Sounds/Attenuation/Fort_GadgetActivateNonBinaural_Att
 maxDistance=4000
 soundClassObject=&Fort_WeaponMelee
 volumeMultiplier=0.3
}
