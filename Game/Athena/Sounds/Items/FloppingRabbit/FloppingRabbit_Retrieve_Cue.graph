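# Sound cue for the FloppingRabbit item's retrieve sound in Athena.
# A single wave player plays the FloppingRabbit_Retrieve wave, slightly boosted in volume, under the gadgets and items sound class.
# It uses a shared non-binaural gadget-activation attenuation asset.
# Overridden concurrency limits it to one instance per owner and prevents a new play while one is active.
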SoundCue /Game/Athena/Sounds/Items/FloppingRabbit/FloppingRabbit_Retrieve_Cue
root=WavePlayer_0
nodes [{id=WavePlayer_0, properties {soundWaveAssetPtr=&/FloppingRabbit_Retrieve}}]
properties {
 attenuationSettings=/Game/Sounds/Attenuation/Fort_GadgetActivateNonBinaural_Att
 bOverrideConcurrency=true
 maxDistance=4000
 soundClassObject=&Fort_Gadgets_Items
 volumeMultiplier=1.2
 concurrencyOverrides {bLimitToOwner=true, maxCount=1, resolutionRule=PreventNew}
}
