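# Sound cue for the Flopping Rabbit item's "tell" sound in Athena, classed under gadgets and items.
# A random node picks one of three equally weighted tell waves.
# A modulator above it applies a slight random upward pitch shift for variation.
# Spatialization uses the Flopping Rabbit hit attenuation asset.
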SoundCue /Game/Athena/Sounds/Items/FloppingRabbit/FloppingRabbit_Tell_Cue
root=Modulator_0
nodes [
 {id=Modulator_0, children [Random_1], properties {pitchMax=1.1, pitchMin=1}}
 {id=Random_1, children [WavePlayer_0,WavePlayer_4,WavePlayer_5], properties {weights [1,1,1]}}
 {id=WavePlayer_0, properties {soundWaveAssetPtr=&/FloppingRabbit_Tell_01}}
 {id=WavePlayer_4, properties {soundWaveAssetPtr=&/FloppingRabbit_Tell_02}}
 {id=WavePlayer_5, properties {soundWaveAssetPtr=&/FloppingRabbit_Tell_03}}
]
properties {
 attenuationSettings=/Game/Sounds/Attenuation/Fort_FloppingRabbitHit_Att
 maxDistance=5300
 soundClassObject=&Fort_Gadgets_Items
}
